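# Physics asset for the Monarch cape accessory, defining simulated cape chains and capsule colliders for torso, arms and legs.
# Four six-segment chains have limited swing and twist: shoulder chains hang from upper arms via MSTR and look-at bodies, middle chains from backpack_root on spine_05.
# Angularly free cross-constraints tie adjacent chains at segments 02–05, linking them like a sheet.
# A large, low-set, unconstrained sphere, dyn_crouch_collision, provides an extra collider, apparently for crouching.
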PhysicsAsset /BRCosmetics/Accessories/FORT_Capes/Cape_F_MED_Monarch/Meshes/F_MED_Monarch_new_cape_Physics
bodies {
 backpack_root [{shapeType=Capsule, length=0.505, radius=0.505, shapeName=backpack_root_capsule, rotation=<rot 0 -90 0>}]
 calf_l [{shapeType=Capsule, length=33.9358, radius=12.9685, shapeName=calf_l_capsule, center=<-30 0 0>, rotation=<rot 90 0 0>}]
 calf_r [{shapeType=Capsule, length=33.9358, radius=12.9685, shapeName=calf_r_capsule, center=<30 0 0>, rotation=<rot 90 0 0>}]
 clavicle_l [{shapeType=Capsule, length=5.5404, radius=5.6421, shapeName=clavicle_l_capsule, center=<7 0 0>, rotation=<rot 90 0 0>}]
 clavicle_r [{shapeType=Capsule, length=5.5404, radius=5.6421, shapeName=clavicle_r_capsule, center=<-7 0 0>, rotation=<rot 90 0 0>}]
 dyn_cape_mid_l_01 [{shapeType=Capsule, length=1, radius=7.2298, shapeName=dyn_cape_mid_l_01_capsule, center=<5 0 5>, rotation=<rot 90 0 0>}]
 dyn_cape_mid_l_02 [{shapeType=Capsule, length=6.8, radius=8.71, shapeName=dyn_cape_mid_l_02_capsule, center=<7.6305 0 5>, rotation=<rot 90 0 0>}]
 dyn_cape_mid_l_03 [{shapeType=Capsule, length=5, radius=10, shapeName=dyn_cape_mid_l_03_capsule, center=<9.1537 0 4.5>, rotation=<rot 90 0 0>}]
 dyn_cape_mid_l_04 [{shapeType=Capsule, length=7.1609, radius=10, shapeName=dyn_cape_mid_l_04_capsule, center=<10.2777 0 6.5>, rotation=<rot 90 0 0>}]
 dyn_cape_mid_l_05 [{shapeType=Capsule, length=5.6513, radius=11, shapeName=dyn_cape_mid_l_05_capsule, center=<11.3076 0 7.5>, rotation=<rot 90 0 0>}]
 dyn_cape_mid_l_06 [{shapeType=Capsule, length=10.8593, radius=9.6886, shapeName=dyn_cape_mid_l_06_capsule, center=<9.2805 0 5>, rotation=<rot 0 0 90>}]
 dyn_cape_mid_l_look_at [{shapeType=Capsule, length=0.505, radius=0.505, shapeName=dyn_cape_mid_l_look_at_capsule, rotation=<rot 0 -90 0>}]
 dyn_cape_mid_r_01 [{shapeType=Capsule, length=1, radius=7.2298, shapeName=dyn_cape_mid_r_01_capsule, center=<6.4214 0 5>, rotation=<rot 90 0 0>}]
 dyn_cape_mid_r_02 [{shapeType=Capsule, length=6.791, radius=8.71, shapeName=dyn_cape_mid_r_02_capsule, center=<7.4493 0 5>, rotation=<rot 90 0 0>}]
 dyn_cape_mid_r_03 [{shapeType=Capsule, length=5, radius=10, shapeName=dyn_cape_mid_r_03_capsule, center=<9.2752 0 4.5>, rotation=<rot 90 0 0>}]
 dyn_cape_mid_r_04 [{shapeType=Capsule, length=7.1609, radius=10, shapeName=dyn_cape_mid_r_04_capsule, center=<10.4626 0 6.5>, rotation=<rot 90 0 0>}]
 dyn_cape_mid_r_05 [{shapeType=Capsule, length=5.6513, radius=11, shapeName=dyn_cape_mid_r_05_capsule, center=<11.1715 0 7.5>, rotation=<rot 90 0 0>}]
 dyn_cape_mid_r_06 [{shapeType=Capsule, length=11.1928, radius=9.9128, shapeName=dyn_cape_mid_r_06_capsule, center=<9.6316 0 5>, rotation=<rot 0 0 90>}]
 dyn_cape_mid_r_look_at [{shapeType=Capsule, length=0.505, radius=0.505, shapeName=dyn_cape_mid_r_look_at_capsule, rotation=<rot 0 -90 0>}]
 dyn_cape_shoulder_l_01 [{shapeType=Capsule, length=0.5028, radius=7.7276, shapeName=dyn_cape_shoulder_l_01_capsule, center=<5 -5 0>, rotation=<rot 90 0 0>}]
 dyn_cape_shoulder_l_02 [{shapeType=Capsule, length=3.85, radius=9.2078, shapeName=dyn_cape_shoulder_l_02_capsule, center=<7.7013 0 5>, rotation=<rot 90 0 0>}]
 dyn_cape_shoulder_l_03 [{shapeType=Capsule, length=5.3645, radius=10, shapeName=dyn_cape_shoulder_l_03_capsule, center=<9.43 0 4.5>, rotation=<rot 90 0 0>}]
 dyn_cape_shoulder_l_04 [{shapeType=Capsule, length=6.1644, radius=10, shapeName=dyn_cape_shoulder_l_04_capsule, center=<10.6519 0 6.5>, rotation=<rot 90 0 0>}]
 dyn_cape_shoulder_l_05 [{shapeType=Capsule, length=5.1536, radius=11, shapeName=dyn_cape_shoulder_l_05_capsule, center=<11.3744 0 7.5>, rotation=<rot 90 0 0>}]
 dyn_cape_shoulder_l_06 [{shapeType=Capsule, length=10.8593, radius=9.6886, shapeName=dyn_cape_shoulder_l_06_capsule, center=<9.8625 0 5>, rotation=<rot 0 0 90>}]
 dyn_cape_shoulder_l_look_at [{shapeType=Capsule, length=0.505, radius=0.505, shapeName=dyn_cape_shoulder_l_look_at_capsule, rotation=<rot 0 -90 0>}]
 dyn_cape_shoulder_r_01 [{shapeType=Capsule, length=0.5923, radius=7.6376, shapeName=dyn_cape_shoulder_r_01_capsule, center=<6.8698 0 5>, rotation=<rot 90 0 0>}]
 dyn_cape_shoulder_r_02 [{shapeType=Capsule, length=6.3833, radius=9.1177, shapeName=dyn_cape_shoulder_r_02_capsule, center=<7.9297 0 5>, rotation=<rot 90 0 0>}]
 dyn_cape_shoulder_r_03 [{shapeType=Capsule, length=5.4546, radius=10, shapeName=dyn_cape_shoulder_r_03_capsule, center=<9.704 0 4.5>, rotation=<rot 90 0 0>}]
 dyn_cape_shoulder_r_04 [{shapeType=Capsule, length=6.2545, radius=10, shapeName=dyn_cape_shoulder_r_04_capsule, center=<10.6686 0 6.5>, rotation=<rot 90 0 0>}]
 dyn_cape_shoulder_r_05 [{shapeType=Capsule, length=5.2436, radius=11, shapeName=dyn_cape_shoulder_r_05_capsule, center=<11.3744 0 7.5>, rotation=<rot 90 0 0>}]
 dyn_cape_shoulder_r_06 [{shapeType=Capsule, length=11.1928, radius=9.9128, shapeName=dyn_cape_shoulder_r_06_capsule, center=<9.8 0 5>, rotation=<rot 0 0 90>}]
 dyn_cape_shoulder_r_look_at [{shapeType=Capsule, length=0.505, radius=0.505, shapeName=dyn_cape_shoulder_r_look_at_capsule, rotation=<rot 0 -90 0>}]
 dyn_crouch_collision [{shapeType=Sphere, radius=300, shapeName=dyn_crouch_collision_sphere, center=<0 -30 -295>}]
 dyn_l_cape_shoulder_MSTR [{shapeType=Capsule, length=0.505, radius=0.505, shapeName=dyn_l_cape_shoulder_MSTR_capsule, rotation=<rot 0 -90 0>}]
 dyn_r_cape_shoulder_MSTR [{shapeType=Capsule, length=0.505, radius=0.505, shapeName=dyn_r_cape_shoulder_MSTR_capsule, rotation=<rot 0 -90 0>}]
 lowerarm_l [{shapeType=Capsule, length=17.025, radius=7.3026, shapeName=lowerarm_l_capsule, center=<7 0 0>, rotation=<rot 90 0 0>}]
 lowerarm_r [{shapeType=Capsule, length=17.2137, radius=6.3863, shapeName=lowerarm_r_capsule, center=<-7 0 0>, rotation=<rot 90 0 0>}]
 pelvis [
  {shapeType=Capsule, length=12.4929, radius=12.6148, shapeName=pelvis_capsule, center=<-0.8652 -0.5292 0>}
  {shapeType=Capsule, length=4.9549, radius=14.7729, shapeName=pelvis_capsule, center=<-11.9052 0.331 0>}
  {shapeType=Capsule, length=8.7262, radius=13.9019, shapeName=pelvis_capsule, center=<-9.4188 0 10.8825>, rotation=<rot 80 -0.0001 0>}
 ]
 spine_01 [{shapeType=Capsule, length=5.9158, radius=12.0639, shapeName=spine_01_capsule, center=<0 0.4899 0>}]
 spine_02 [{shapeType=Capsule, length=6.2629, radius=10.7649, shapeName=spine_02_capsule, center=<0 -0.2876 0>}]
 spine_03 [{shapeType=Capsule, length=3.3342, radius=9.9923, shapeName=spine_03_capsule, center=<0 0.0099 0>}]
 spine_04 [{shapeType=Capsule, length=9.9324, radius=8.482, shapeName=spine_04_capsule, center=<0 0.6118 0>}]
 spine_05 [{shapeType=Capsule, length=14, radius=10, shapeName=spine_05_capsule, center=<0 1.5 0>}]
 thigh_l [{shapeType=Capsule, length=27.8994, radius=12.0305, shapeName=thigh_l_capsule, center=<-25 0 0>, rotation=<rot 90 0 0>}]
 thigh_r [{shapeType=Capsule, length=30.2853, radius=11.6185, shapeName=thigh_r_capsule, center=<25 0 0>, rotation=<rot 90 0 0>}]
 upperarm_l [{shapeType=Capsule, length=16.898, radius=7.9184, shapeName=upperarm_l_capsule, center=<12 2.7231 0>, rotation=<rot 90 0 0>}]
 upperarm_r [{shapeType=Capsule, length=19.415, radius=6.6952, shapeName=upperarm_r_capsule, center=<-12 0 0>, rotation=<rot 90 0 0>}]
}
constraints [
 {
  bone1Name=spine_01
  bone2Name=pelvis
  swing1LimitDegrees=45
  swing1Motion=Free
  swing2LimitDegrees=45
  swing2Motion=Free
  twistLimitDegrees=45
  twistMotion=Free
 }
 {
  bone1Name=spine_02
  bone2Name=spine_01
  swing1LimitDegrees=45
  swing1Motion=Free
  swing2LimitDegrees=45
  swing2Motion=Free
  twistLimitDegrees=45
  twistMotion=Free
 }
 {
  bone1Name=spine_04
  bone2Name=spine_03
  swing1LimitDegrees=45
  swing1Motion=Free
  swing2LimitDegrees=45
  swing2Motion=Free
  twistLimitDegrees=45
  twistMotion=Free
 }
 {
  bone1Name=spine_03
  bone2Name=spine_02
  swing1LimitDegrees=45
  swing1Motion=Free
  swing2LimitDegrees=45
  swing2Motion=Free
  twistLimitDegrees=45
  twistMotion=Free
 }
 {
  bone1Name=thigh_r
  bone2Name=pelvis
  swing1LimitDegrees=45
  swing1Motion=Free
  swing2LimitDegrees=45
  swing2Motion=Free
  twistLimitDegrees=45
  twistMotion=Free
 }
 {
  bone1Name=thigh_l
  bone2Name=pelvis
  swing1LimitDegrees=45
  swing1Motion=Free
  swing2LimitDegrees=45
  swing2Motion=Free
  twistLimitDegrees=45
  twistMotion=Free
 }
 {
  bone1Name=spine_05
  bone2Name=spine_04
  swing1LimitDegrees=45
  swing1Motion=Free
  swing2LimitDegrees=45
  swing2Motion=Free
  twistLimitDegrees=45
  twistMotion=Free
 }
 {
  bone1Name=clavicle_l
  bone2Name=spine_05
  swing1LimitDegrees=45
  swing1Motion=Free
  swing2LimitDegrees=45
  swing2Motion=Free
  twistLimitDegrees=45
  twistMotion=Free
 }
 {
  bone1Name=upperarm_l
  bone2Name=clavicle_l
  swing1LimitDegrees=45
  swing1Motion=Free
  swing2LimitDegrees=45
  swing2Motion=Free
  twistLimitDegrees=45
  twistMotion=Free
 }
 {
  bone1Name=lowerarm_l
  bone2Name=upperarm_l
  swing1LimitDegrees=45
  swing1Motion=Free
  swing2LimitDegrees=45
  swing2Motion=Free
  twistLimitDegrees=45
  twistMotion=Free
 }
 {
  bone1Name=clavicle_r
  bone2Name=spine_05
  swing1LimitDegrees=45
  swing1Motion=Free
  swing2LimitDegrees=45
  swing2Motion=Free
  twistLimitDegrees=45
  twistMotion=Free
 }
 {
  bone1Name=upperarm_r
  bone2Name=clavicle_r
  swing1LimitDegrees=45
  swing1Motion=Free
  swing2LimitDegrees=45
  swing2Motion=Free
  twistLimitDegrees=45
  twistMotion=Free
 }
 {
  bone1Name=lowerarm_r
  bone2Name=upperarm_r
  swing1LimitDegrees=45
  swing1Motion=Free
  swing2LimitDegrees=45
  swing2Motion=Free
  twistLimitDegrees=45
  twistMotion=Free
 }
 {
  bone1Name=calf_l
  bone2Name=thigh_l
  swing1LimitDegrees=45
  swing1Motion=Free
  swing2LimitDegrees=45
  swing2Motion=Free
  twistLimitDegrees=45
  twistMotion=Free
 }
 {
  bone1Name=calf_r
  bone2Name=thigh_r
  swing1LimitDegrees=45
  swing1Motion=Free
  swing2LimitDegrees=45
  swing2Motion=Free
  twistLimitDegrees=45
  twistMotion=Free
 }
 {
  bone1Name=dyn_l_cape_shoulder_MSTR
  bone2Name=upperarm_l
  swing1LimitDegrees=45
  swing1Motion=Free
  swing2LimitDegrees=45
  swing2Motion=Free
  twistLimitDegrees=45
  twistMotion=Free
 }
 {
  bone1Name=dyn_cape_shoulder_l_look_at
  bone2Name=dyn_l_cape_shoulder_MSTR
  swing1LimitDegrees=45
  swing1Motion=Free
  swing2LimitDegrees=45
  swing2Motion=Free
  twistLimitDegrees=45
  twistMotion=Free
 }
 {bone1Name=dyn_cape_shoulder_l_01, bone2Name=dyn_cape_shoulder_l_look_at, swing1LimitDegrees=35, swing2LimitDegrees=15, twistLimitDegrees=15}
 {bone1Name=dyn_cape_shoulder_l_02, bone2Name=dyn_cape_shoulder_l_01, swing1LimitDegrees=15, swing2LimitDegrees=35, twistLimitDegrees=15}
 {bone1Name=dyn_cape_shoulder_l_03, bone2Name=dyn_cape_shoulder_l_02, swing1LimitDegrees=25, swing2LimitDegrees=25, twistLimitDegrees=15}
 {bone1Name=dyn_cape_shoulder_l_04, bone2Name=dyn_cape_shoulder_l_03, swing1LimitDegrees=45, swing2LimitDegrees=45, twistLimitDegrees=15}
 {bone1Name=dyn_cape_shoulder_l_05, bone2Name=dyn_cape_shoulder_l_04, swing1LimitDegrees=45, swing2LimitDegrees=45, twistLimitDegrees=15}
 {bone1Name=dyn_cape_shoulder_l_06, bone2Name=dyn_cape_shoulder_l_05, swing1LimitDegrees=45, swing2LimitDegrees=45, twistLimitDegrees=15}
 {
  bone1Name=dyn_r_cape_shoulder_MSTR
  bone2Name=upperarm_r
  swing1LimitDegrees=45
  swing1Motion=Free
  swing2LimitDegrees=45
  swing2Motion=Free
  twistLimitDegrees=45
  twistMotion=Free
 }
 {
  bone1Name=dyn_cape_shoulder_r_look_at
  bone2Name=dyn_r_cape_shoulder_MSTR
  swing1LimitDegrees=45
  swing1Motion=Free
  swing2LimitDegrees=45
  swing2Motion=Free
  twistLimitDegrees=45
  twistMotion=Free
 }
 {bone1Name=dyn_cape_shoulder_r_01, bone2Name=dyn_cape_shoulder_r_look_at, swing1LimitDegrees=15, swing2LimitDegrees=35, twistLimitDegrees=15}
 {bone1Name=dyn_cape_shoulder_r_02, bone2Name=dyn_cape_shoulder_r_01, swing1LimitDegrees=15, swing2LimitDegrees=35, twistLimitDegrees=15}
 {bone1Name=dyn_cape_shoulder_r_03, bone2Name=dyn_cape_shoulder_r_02, swing1LimitDegrees=25, swing2LimitDegrees=25, twistLimitDegrees=15}
 {bone1Name=dyn_cape_shoulder_r_04, bone2Name=dyn_cape_shoulder_r_03, swing1LimitDegrees=45, swing2LimitDegrees=45, twistLimitDegrees=15}
 {bone1Name=dyn_cape_shoulder_r_05, bone2Name=dyn_cape_shoulder_r_04, swing1LimitDegrees=45, swing2LimitDegrees=45, twistLimitDegrees=15}
 {bone1Name=dyn_cape_shoulder_r_06, bone2Name=dyn_cape_shoulder_r_05, swing1LimitDegrees=45, swing2LimitDegrees=45, twistLimitDegrees=15}
 {
  bone1Name=dyn_cape_mid_l_look_at
  bone2Name=backpack_root
  swing1LimitDegrees=45
  swing1Motion=Free
  swing2LimitDegrees=45
  swing2Motion=Free
  twistLimitDegrees=45
  twistMotion=Free
 }
 {bone1Name=dyn_cape_mid_l_01, bone2Name=dyn_cape_mid_l_look_at, swing1LimitDegrees=45, swing2LimitDegrees=45, twistLimitDegrees=15}
 {bone1Name=dyn_cape_mid_l_02, bone2Name=dyn_cape_mid_l_01, swing1LimitDegrees=45, swing2LimitDegrees=45, twistLimitDegrees=15}
 {bone1Name=dyn_cape_mid_l_03, bone2Name=dyn_cape_mid_l_02, swing1LimitDegrees=45, swing2LimitDegrees=45, twistLimitDegrees=15}
 {bone1Name=dyn_cape_mid_l_04, bone2Name=dyn_cape_mid_l_03, swing1LimitDegrees=45, swing2LimitDegrees=45, twistLimitDegrees=15}
 {bone1Name=dyn_cape_mid_l_05, bone2Name=dyn_cape_mid_l_04, swing1LimitDegrees=45, swing2LimitDegrees=45, twistLimitDegrees=15}
 {bone1Name=dyn_cape_mid_l_06, bone2Name=dyn_cape_mid_l_05, swing1LimitDegrees=45, swing2LimitDegrees=45, twistLimitDegrees=15}
 {
  bone1Name=dyn_cape_mid_r_look_at
  bone2Name=backpack_root
  swing1LimitDegrees=45
  swing1Motion=Free
  swing2LimitDegrees=45
  swing2Motion=Free
  twistLimitDegrees=45
  twistMotion=Free
 }
 {bone1Name=dyn_cape_mid_r_01, bone2Name=dyn_cape_mid_r_look_at, swing1LimitDegrees=45, swing2LimitDegrees=45, twistLimitDegrees=15}
 {bone1Name=dyn_cape_mid_r_02, bone2Name=dyn_cape_mid_r_01, swing1LimitDegrees=45, swing2LimitDegrees=45, twistLimitDegrees=15}
 {bone1Name=dyn_cape_mid_r_03, bone2Name=dyn_cape_mid_r_02, swing1LimitDegrees=45, swing2LimitDegrees=45, twistLimitDegrees=15}
 {bone1Name=dyn_cape_mid_r_04, bone2Name=dyn_cape_mid_r_03, swing1LimitDegrees=45, swing2LimitDegrees=45, twistLimitDegrees=15}
 {bone1Name=dyn_cape_mid_r_05, bone2Name=dyn_cape_mid_r_04, swing1LimitDegrees=45, swing2LimitDegrees=45, twistLimitDegrees=15}
 {bone1Name=dyn_cape_mid_r_06, bone2Name=dyn_cape_mid_r_05, swing1LimitDegrees=45, swing2LimitDegrees=45, twistLimitDegrees=15}
 {
  bone1Name=backpack_root
  bone2Name=spine_05
  swing1LimitDegrees=45
  swing1Motion=Free
  swing2LimitDegrees=45
  swing2Motion=Free
  twistLimitDegrees=45
  twistMotion=Free
 }
 {
  bone1Name=dyn_cape_mid_l_02
  bone2Name=dyn_cape_mid_r_02
  swing1LimitDegrees=45
  swing1Motion=Free
  swing2LimitDegrees=45
  swing2Motion=Free
  twistLimitDegrees=45
  twistMotion=Free
 }
 {
  bone1Name=dyn_cape_mid_l_05
  bone2Name=dyn_cape_mid_r_05
  swing1LimitDegrees=45
  swing1Motion=Free
  swing2LimitDegrees=45
  swing2Motion=Free
  twistLimitDegrees=45
  twistMotion=Free
 }
 {
  bone1Name=dyn_cape_mid_l_03
  bone2Name=dyn_cape_mid_r_03
  swing1LimitDegrees=45
  swing1Motion=Free
  swing2LimitDegrees=45
  swing2Motion=Free
  twistLimitDegrees=45
  twistMotion=Free
 }
 {
  bone1Name=dyn_cape_shoulder_l_02
  bone2Name=dyn_cape_mid_l_02
  swing1LimitDegrees=45
  swing1Motion=Free
  swing2LimitDegrees=45
  swing2Motion=Free
  twistLimitDegrees=45
  twistMotion=Free
 }
 {
  bone1Name=dyn_cape_shoulder_l_03
  bone2Name=dyn_cape_mid_l_03
  swing1LimitDegrees=45
  swing1Motion=Free
  swing2LimitDegrees=45
  swing2Motion=Free
  twistLimitDegrees=45
  twistMotion=Free
 }
 {
  bone1Name=dyn_cape_shoulder_l_05
  bone2Name=dyn_cape_mid_l_05
  swing1LimitDegrees=45
  swing1Motion=Free
  swing2LimitDegrees=45
  swing2Motion=Free
  twistLimitDegrees=45
  twistMotion=Free
 }
 {
  bone1Name=dyn_cape_shoulder_r_02
  bone2Name=dyn_cape_mid_r_02
  swing1LimitDegrees=45
  swing1Motion=Free
  swing2LimitDegrees=45
  swing2Motion=Free
  twistLimitDegrees=45
  twistMotion=Free
 }
 {
  bone1Name=dyn_cape_shoulder_r_05
  bone2Name=dyn_cape_mid_r_05
  swing1LimitDegrees=45
  swing1Motion=Free
  swing2LimitDegrees=45
  swing2Motion=Free
  twistLimitDegrees=45
  twistMotion=Free
 }
 {
  bone1Name=dyn_cape_shoulder_r_04
  bone2Name=dyn_cape_mid_r_04
  swing1LimitDegrees=45
  swing1Motion=Free
  swing2LimitDegrees=45
  swing2Motion=Free
  twistLimitDegrees=45
  twistMotion=Free
 }
 {
  bone1Name=dyn_cape_mid_l_04
  bone2Name=dyn_cape_mid_r_04
  swing1LimitDegrees=45
  swing1Motion=Free
  swing2LimitDegrees=45
  swing2Motion=Free
  twistLimitDegrees=45
  twistMotion=Free
 }
 {
  bone1Name=dyn_cape_shoulder_r_03
  bone2Name=dyn_cape_mid_r_03
  swing1LimitDegrees=45
  swing1Motion=Free
  swing2LimitDegrees=45
  swing2Motion=Free
  twistLimitDegrees=45
  twistMotion=Free
 }
 {
  bone1Name=dyn_cape_shoulder_l_04
  bone2Name=dyn_cape_mid_l_04
  swing1LimitDegrees=45
  swing1Motion=Free
  swing2LimitDegrees=45
  swing2Motion=Free
  twistLimitDegrees=45
  twistMotion=Free
 }
]
