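# Animation Blueprint for the Monarch cape cosmetic, copying the attached parent's pose and adding cape motion.
# Slope, sim-space and state-based bone offsets feed a cached pose, branching into RigidBody simulation (front-end or gameplay tuning) or constraint-plus-trail-controller fallback, chosen by bRigidBodyEnabled.
# The event graph drives gravity, sim-space rotation, velocity scales and trail alpha from movement, skydiving, aiming and parachute state.
# The DBNO blend uses trailCache on both inputs, leaving dbnoCache out of the output.
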AnimBlueprint /BRCosmetics/Accessories/FORT_Capes/Cape_F_MED_Monarch/Meshes/F_MED_Monarch_Cape_AnimBP
target_skeleton=&/F_MED_Monarch_Cape_Skeleton
graphs [
 AnimGraph {nodes [
   Component To Local()
   n1 = "Constraint - dyn_cape_mid_l_look_at (PP)"!(ComponentPose, Alpha=1, ConstraintWeights_0=0.5, ConstraintWeights_1:float=0.5) -> (Pose) {
    alphaBoolBlend {blendOption=Linear}
    boneToModify {boneName=dyn_cape_mid_l_look_at}
    constraintSetup [{targetBone {boneName=backpack_root}}, {targetBone {boneName=dyn_l_cape_shoulder_MSTR}}]
    constraintWeights [1,1]
   }
   "Constraint - dyn_cape_mid_r_look_at (PP)"!(ComponentPose=@n1, Alpha=1, ConstraintWeights_0=0.5, ConstraintWeights_1:float=0.5) -> (Pose) {
    alphaBoolBlend {blendOption=Linear}
    boneToModify {boneName=dyn_cape_mid_r_look_at}
    constraintSetup [{targetBone {boneName=backpack_root}}, {targetBone {boneName=dyn_r_cape_shoulder_MSTR}}]
    constraintWeights [1,1]
   }
   n2 = "Constraint - cape_00_r (PP)"!(ComponentPose, Alpha=1, ConstraintWeights_0=0.28, ConstraintWeights_1:float=0.72) -> (Pose) {
    alphaBoolBlend {blendOption=Linear}
    boneToModify {boneName=cape_00_r}
    constraintSetup [{targetBone {boneName=spine_05}}, {targetBone {boneName=clavicle_r}}]
    constraintWeights [1,1]
   }
   "Constraint - cape_00_l (PP)"!(ComponentPose=@n2, Alpha=1, ConstraintWeights_0=0.2, ConstraintWeights_1:float=0.8) -> (Pose) {
    alphaBoolBlend {blendOption=Linear}
    boneToModify {boneName=cape_00_l}
    constraintSetup [{targetBone {boneName=spine_05}}, {targetBone {boneName=clavicle_l}}]
    constraintWeights [1,1]
   }
   n3 = Look At - Bone: cape_01_l(Alpha=1) {
    alphaBoolBlend {blendOption=Linear}
    boneToModify {boneName=cape_01_l}
    lookAtLocation=<0 -50 -18>
    lookAtTarget {boneReference {boneName=pelvis}}
    lookAt_Axis {axis=<1 0>}
   }
   Look At - Bone: cape_01_r(ComponentPose=@n3, Alpha=1) {
    alphaBoolBlend {blendOption=Linear}
    boneToModify {boneName=cape_01_r}
    lookAtLocation=<0 -50 18>
    lookAtTarget {boneReference {boneName=pelvis}}
    lookAt_Axis {axis=<1 0>}
   }
   n4 = Look At - Bone: cape_02_l(Alpha=1) {
    alphaBoolBlend {blendOption=Linear}
    boneToModify {boneName=cape_02_l}
    lookAtLocation=<0 -30 -25>
    lookAtTarget {boneReference {boneName=pelvis}}
    lookAt_Axis {axis=<1 0>}
   }
   n5 = Look At - Bone: cape_02_r(ComponentPose=@n4, Alpha=1) {
    alphaBoolBlend {blendOption=Linear}
    boneToModify {boneName=cape_02_r}
    lookAtLocation=<0 -30 25>
    lookAtTarget {boneReference {boneName=pelvis}}
    lookAt_Axis {axis=<1 0>}
   }
   n6 = Look At - Bone: cape_mid_01_l(ComponentPose=@n5, Alpha=1) {
    alphaBoolBlend {blendOption=Linear}
    boneToModify {boneName=cape_mid_01_l}
    lookAtLocation=<0 -30 -5>
    lookAtTarget {boneReference {boneName=pelvis}}
    lookAt_Axis {axis=<1 0>}
   }
   Look At - Bone: cape_mid_01_r(ComponentPose=@n6, Alpha=1) {
    alphaBoolBlend {blendOption=Linear}
    boneToModify {boneName=cape_mid_01_r}
    lookAtLocation=<0 -30 5>
    lookAtTarget {boneReference {boneName=pelvis}}
    lookAt_Axis {axis=<1 0>}
   }
   "Transform (Modify) Bone - Bone: dyn_upperarm_twist_cape"(Alpha=Get armForwardUpR() -> (armForwardUpR:double), Translation="1.5,0,0", Rotation="0,0,0", Scale="1,1,1") {translationMode=BMM_Additive, alphaBoolBlend {blendOption=Linear}, boneToModify {boneName=dyn_upperarm_twist_cape}, translation {x=1.5}}
   n7 = "Transform (Modify) Bone - Bone: clavicle_cape_r"!(ComponentPose, bAlphaBoolEnabled:bool=True, Translation="0,10,0", Rotation="0.000,0.000,20", Scale="1,1,1") -> (Pose) {
    alphaInputType=Bool
    rotationMode=BMM_Additive
    translationMode=BMM_Additive
    alphaBoolBlend {blendInTime=0.3, blendOption=Linear, blendOutTime=0.3}
    boneToModify {boneName=clavicle_cape_r}
    rotation {roll=20}
    translation {y=10}
   }
   "Transform (Modify) Bone - Bone: clavicle_cape_l"!(ComponentPose=@n7, bAlphaBoolEnabled:bool=True, Translation="0,0,0", Rotation="0.000,0.000,20", Scale="1,1,1") -> (Pose) {
    alphaInputType=Bool
    rotationMode=BMM_Additive
    alphaBoolBlend {blendInTime=0.3, blendOption=Linear, blendOutTime=0.3}
    boneToModify {boneName=clavicle_cape_l}
    rotation {roll=20}
   }
   n8 = "Transform (Modify) Bone - Bone: clavicle_cape_r"(bAlphaBoolEnabled=True, Translation="0,-2,0", Rotation="0.000,0.000,35", Scale="1,1,1") {
    alphaInputType=Bool
    rotationMode=BMM_Additive
    alphaBoolBlend {blendInTime=0.3, blendOption=Linear, blendOutTime=0.3}
    boneToModify {boneName=clavicle_cape_r}
    rotation {roll=35}
    translation {y=-2}
   }
   "Transform (Modify) Bone - Bone: clavicle_cape_l"(ComponentPose=@n8, bAlphaBoolEnabled=True, Translation="0,-2,0", Rotation="0.000,0.000,35", Scale="1,1,1") {
    alphaInputType=Bool
    rotationMode=BMM_Additive
    alphaBoolBlend {blendInTime=0.3, blendOption=Linear, blendOutTime=0.3}
    boneToModify {boneName=clavicle_cape_l}
    rotation {roll=35}
    translation {y=-2}
   }
   "Transform (Modify) Bone - Bone: backpack_root"!(ComponentPose, bAlphaBoolEnabled:bool=True, Translation="0,-2,0", Rotation="0.000,0.000,40", Scale="1,1,1") -> (Pose) {
    alphaInputType=Bool
    rotationMode=BMM_Additive
    alphaBoolBlend {blendInTime=0.3, blendOption=Linear, blendOutTime=0.3}
    boneToModify {boneName=backpack_root}
    rotation {roll=40}
    translation {y=-2}
   }
   "Transform (Modify) Bone - Bone: dyn_l_cape_shoulder_MSTR"(Alpha=1, Translation="0,0,0", Rotation="0,-15,15", Scale="1,1,1") {
    rotationMode=BMM_Additive
    rotationSpace=BCS_BoneSpace
    translationMode=BMM_Additive
    translationSpace=BCS_BoneSpace
    alphaBoolBlend {blendOption=Linear}
    boneToModify {boneName=dyn_l_cape_shoulder_MSTR}
    rotation {roll=15, yaw=-15}
   }
   n9 = "Transform (Modify) Bone - Bone: backpack_root"(bAlphaBoolEnabled=True, Translation="0,0,0", Rotation="0.000,0.000,35", Scale="1,1,1") {
    alphaInputType=Bool
    rotationMode=BMM_Additive
    alphaBoolBlend {blendInTime=0.3, blendOption=Linear, blendOutTime=0.3}
    boneToModify {boneName=backpack_root}
    rotation {roll=35}
   }
   n10 = "Transform (Modify) Bone - Bone: cape_00_l"!(ComponentPose=@n9, bAlphaBoolEnabled:bool=True, Translation="0,0,0", Rotation="0.000,0.000,35", Scale="1,1,1") -> (Pose) {
    alphaInputType=Bool
    rotationMode=BMM_Additive
    alphaBoolBlend {blendInTime=0.3, blendOption=Linear, blendOutTime=0.3}
    boneToModify {boneName=cape_00_l}
    rotation {roll=35}
   }
   n11 = "Transform (Modify) Bone - Bone: cape_mid_01_r"!(ComponentPose=@n10, bAlphaBoolEnabled:bool=True, Translation="0,0,0", Rotation="0.000,0.000,35", Scale="1,1,1") -> (Pose) {
    alphaInputType=Bool
    rotationMode=BMM_Additive
    alphaBoolBlend {blendInTime=0.3, blendOption=Linear, blendOutTime=0.3}
    boneToModify {boneName=cape_mid_01_r}
    rotation {roll=35}
   }
   "Transform (Modify) Bone - Bone: cape_00_r"!(ComponentPose=@n11, bAlphaBoolEnabled:bool=True, Translation="0,0,0", Rotation="0.000,0.000,35", Scale="1,1,1") -> (Pose) {
    alphaInputType=Bool
    rotationMode=BMM_Additive
    alphaBoolBlend {blendInTime=0.3, blendOption=Linear, blendOutTime=0.3}
    boneToModify {boneName=cape_00_r}
    rotation {roll=35}
   }
   n12 = Get GravityOverride()
   n13 = Get ADS Accel() -> (ADS Accel:Vector)
   n14 = Get ADS Vel() -> (ADS Vel:Vector)
   RigidBody!(
    ComponentPose
    Alpha
    OverrideWorldGravity:Vector=@n12
    ExternalForce
    ComponentLinearAccScale:Vector=@n13
    ComponentLinearVelScale:Vector=@n14
   ) -> (Pose) {
    bClampLinearTranslationLimitToRefPose=true
    bOverrideWorldGravity=true
    overridePhysicsAsset=&/F_MED_Monarch_new_cape_Physics
    simulationSpace=BaseBoneSpace
    alphaBoolBlend {blendOption=Linear}
    baseBoneRef {boneName=pelvis}
    componentAppliedLinearAccClamp=<3000 3000 3000>
    componentLinearAccScale=<0.5 0.3 0.3>
    componentLinearVelScale=<0.5 0.3 0.3>
    overrideWorldGravity {z=-980}
    simSpaceSettings {
     maxAngularAcceleration=1000
     maxAngularVelocity=3
     maxLinearAcceleration=700
     maxLinearVelocity=500
     velocityScaleZ=0
     worldAlpha=1
     externalLinearDragV=<0.2 0.2>
    }
   }
   n15 = Blend Poses by bool(BlendPose_0=Use cached pose 'trailCache'(), BlendPose_1=Use cached pose 'trailCache'(), bActiveValue=Get bIsDBNO()) {blendType=Linear}
   n16 = Use cached pose 'rigidChche'()
   n17 = Get bRigidBodyEnabled()
   n18 = "Transform (Modify) Bone - Bone: backpack_root"(
    ComponentPose=Local To Component(LocalPose=Use cached pose 'trailCache'())
    bAlphaBoolEnabled=True
    Translation="0,5,0"
    Rotation="0.000,0.000,20"
    Scale="1,1,1"
   ) {
    alphaInputType=Bool
    rotationMode=BMM_Additive
    translationMode=BMM_Additive
    alphaBoolBlend {blendInTime=0.3, blendOption=Linear, blendOutTime=0.3}
    boneToModify {boneName=backpack_root}
    rotation {roll=20}
    translation {y=5}
   }
   Output Pose(Result=Blend Poses by bool(
     BlendPose_0=Component To Local(ComponentPose=@n18)
     BlendPose_1=Blend Poses by bool(BlendPose_0=@n16, BlendPose_1=@n15, BlendTime_0=1, bActiveValue=@n17) {blendType=Linear}
     bActiveValue=Get front end no rigid() -> (front end no rigid:bool)
    ) {blendType=Linear}) {layerGroup=None}
   n19 = "Transform (Modify) Bone - Bone: dyn_root_slope"!(
    ComponentPose=Local To Component(LocalPose=Copy Pose From Mesh() {bUseAttachedParent=true})
    bAlphaBoolEnabled:bool=NOT Boolean(A=Get bCanPlayCustomAnimations() -> (bCanPlayCustomAnimations:bool))
    Translation=Get slope_root_trans() -> (slope_root_trans:Vector)
    Rotation=Get slope_foot_rot() -> (slope_foot_rot:Rotator)
    Scale="1,1,1"
   ) -> (Pose) {
    alphaInputType=Bool
    rotationMode=BMM_Replace
    rotationSpace=BCS_WorldSpace
    translationMode=BMM_Replace
    translationSpace=BCS_WorldSpace
    alphaBoolBlend {blendOption=Linear}
    boneToModify {boneName=dyn_root_slope}
   }
   n20 = "Transform (Modify) Bone - Bone: dyn_root_simspace"(ComponentPose=@n19, Alpha=1, Translation="0,0,0", Rotation=Get sim_space_rotation() -> (sim_space_rotation:Rotator), Scale="1,1,1") {
    rotationMode=BMM_Replace
    rotationSpace=BCS_WorldSpace
    translationSpace=BCS_BoneSpace
    alphaBoolBlend {blendOption=Linear}
    alphaScaleBiasClamp {bInterpResult=true, interpSpeedDecreasing=2, interpSpeedIncreasing=2}
    boneToModify {boneName=dyn_root_simspace}
   }
   n21 = "Transform (Modify) Bone - Bone: dyn_cape_tgt"!(ComponentPose=@n20, bAlphaBoolEnabled:bool=True, Translation="0.000,-80,100", Rotation="0,0,0", Scale="1,1,1") -> (Pose) {
    alphaInputType=Bool
    rotationMode=BMM_Additive
    translationMode=BMM_Additive
    translationSpace=BCS_BoneSpace
    alphaBoolBlend {blendInTime=0.3, blendOption=Linear, blendOutTime=0.3}
    boneToModify {boneName=dyn_cape_tgt}
    translation {y=-80, z=100}
   }
   n22 = "Transform (Modify) Bone - Bone: dyn_cape_tgt"(ComponentPose=@n21, bAlphaBoolEnabled=True, Translation="-0.000,0,150.000", Rotation="0,0,0", Scale="1,1,1") {
    alphaInputType=Bool
    rotationMode=BMM_Additive
    translationMode=BMM_Additive
    translationSpace=BCS_BoneSpace
    alphaBoolBlend {blendInTime=0.3, blendOption=Linear, blendOutTime=0.3}
    boneToModify {boneName=dyn_cape_tgt}
    translation {z=150}
   }
   n23 = "Transform (Modify) Bone - Bone: backpack_root"(ComponentPose=@n22, bAlphaBoolEnabled=True, Translation="0,-2,0", Rotation="0.000,0.000,40", Scale="1,1,1") {
    alphaInputType=Bool
    rotationMode=BMM_Additive
    alphaBoolBlend {blendInTime=0.3, blendOption=Linear, blendOutTime=0.3}
    boneToModify {boneName=backpack_root}
    rotation {roll=40}
    translation {y=-2}
   }
   n24 = "Transform (Modify) Bone - Bone: dyn_crouch_collision"!(
    ComponentPose=@n23
    bAlphaBoolEnabled:bool=OR Boolean(A=Get bIsDBNO(), B=Get bIsFalling())
    Translation="0,0,-50"
    Rotation="0,0,0"
    Scale="1,1,1"
   ) -> (Pose) {
    alphaInputType=Bool
    rotationMode=BMM_Additive
    translationMode=BMM_Additive
    translationSpace=BCS_BoneSpace
    alphaBoolBlend {blendInTime=0.1, blendOption=Linear, blendOutTime=0.1}
    boneToModify {boneName=dyn_crouch_collision}
    translation {z=-50}
   }
   n25 = "Transform (Modify) Bone - Bone: backpack_root"(
    ComponentPose=@n24
    bAlphaBoolEnabled=Get skydive rigids() -> (skydive rigids:bool)
    Translation="0,-2,0"
    Rotation="0.000,0.000,-30"
    Scale="1,1,1"
   ) {
    alphaInputType=Bool
    rotationMode=BMM_Additive
    alphaBoolBlend {blendInTime=0.3, blendOption=Linear, blendOutTime=0.3}
    boneToModify {boneName=backpack_root}
    rotation {roll=-30}
    translation {y=-2}
   }
   n26 = Look At - Bone: dyn_cape_shoulder_l_look_at(ComponentPose=@n25, Alpha=1) {
    alphaBoolBlend {blendOption=Linear}
    boneToModify {boneName=dyn_cape_shoulder_l_look_at}
    lookAtLocation {x=0}
    lookAtTarget {boneReference {boneName=dyn_cape_tgt}}
    lookAt_Axis {axis=<1 0>}
   }
   n27 = Look At - Bone: dyn_cape_shoulder_r_look_at(ComponentPose=@n26, Alpha=1) {
    alphaBoolBlend {blendOption=Linear}
    boneToModify {boneName=dyn_cape_shoulder_r_look_at}
    lookAtLocation {x=0}
    lookAtTarget {boneReference {boneName=dyn_cape_tgt}}
    lookAt_Axis {axis=<1 0>}
   }
   Save cached pose 'initialPoseCache'(Pose=Component To Local(ComponentPose=@n27)) {cacheName=initialPoseCache}
   n28 = "Constraint - dyn_cape_shoulder_r_01 (PP)"!(
    ComponentPose=Local To Component(LocalPose=Use cached pose 'initialPoseCache'())
    Alpha=1
    ConstraintWeights_0=0.28
    ConstraintWeights_1:float=0.72
   ) -> (Pose) {
    alphaBoolBlend {blendOption=Linear}
    boneToModify {boneName=dyn_cape_shoulder_r_01}
    constraintSetup [{targetBone {boneName=spine_05}}, {targetBone {boneName=clavicle_r}}]
    constraintWeights [1,1]
   }
   n29 = "Constraint - dyn_cape_shoulder_r_02 (PP)"!(ComponentPose=@n28, Alpha=1, ConstraintWeights_0=0.5, ConstraintWeights_1:float=0.5) -> (Pose) {
    alphaBoolBlend {blendOption=Linear}
    boneToModify {boneName=dyn_cape_shoulder_r_02}
    constraintSetup [{targetBone {boneName=spine_05}}, {targetBone {boneName=clavicle_r}}]
    constraintWeights [1,1]
   }
   n30 = "Constraint - dyn_cape_shoulder_l_01 (PP)"!(ComponentPose=@n29, Alpha=1, ConstraintWeights_0=0.1, ConstraintWeights_1:float=0.9) -> (Pose) {
    alphaBoolBlend {blendOption=Linear}
    boneToModify {boneName=dyn_cape_shoulder_l_01}
    constraintSetup [{targetBone {boneName=spine_05}}, {targetBone {boneName=clavicle_l}}]
    constraintWeights [1,1]
   }
   n31 = "Constraint - dyn_cape_shoulder_l_02 (PP)"!(ComponentPose=@n30, Alpha=1, ConstraintWeights_0=0.5, ConstraintWeights_1:float=0.5) -> (Pose) {
    alphaBoolBlend {blendOption=Linear}
    boneToModify {boneName=dyn_cape_shoulder_l_02}
    constraintSetup [{targetBone {boneName=spine_05}}, {targetBone {boneName=clavicle_l}}]
    constraintWeights [1,1]
   }
   n32 = "Constraint - dyn_cape_shoulder_l_03 (PP)"!(ComponentPose=@n31, Alpha=1, ConstraintWeights_0=1, ConstraintWeights_1:float=0) -> (Pose) {
    alphaBoolBlend {blendOption=Linear}
    boneToModify {boneName=dyn_cape_shoulder_l_03}
    constraintSetup [{targetBone {boneName=spine_05}}, {targetBone {boneName=clavicle_l}}]
    constraintWeights [1,1]
   }
   n33 = "Constraint - dyn_cape_shoulder_r_03 (PP)"!(ComponentPose=@n32, Alpha=1, ConstraintWeights_0=1, ConstraintWeights_1:float=0) -> (Pose) {
    alphaBoolBlend {blendOption=Linear}
    boneToModify {boneName=dyn_cape_shoulder_r_03}
    constraintSetup [{targetBone {boneName=spine_05}}, {targetBone {boneName=clavicle_l}}]
    constraintWeights [1,1]
   }
   n34 = Get SkydivingAlpha() -> (SkydivingAlpha:Rotator)
   n35 = "Transform (Modify) Bone - Bone: dyn_cape_shoulder_l_02"!(ComponentPose=@n33, Alpha=1, Rotation=@n34, Scale="1,1,1") -> (Pose) {
    rotationMode=BMM_Additive
    rotationSpace=BCS_BoneSpace
    translationMode=BMM_Additive
    translationSpace=BCS_BoneSpace
    alphaBoolBlend {blendOption=Linear}
    boneToModify {boneName=dyn_cape_shoulder_l_02}
    translation {y=-3}
   }
   n36 = "Transform (Modify) Bone - Bone: dyn_cape_shoulder_r_02"!(ComponentPose=@n35, Alpha=1, Rotation=@n34, Scale="1,1,1") -> (Pose) {
    rotationMode=BMM_Additive
    rotationSpace=BCS_BoneSpace
    translationMode=BMM_Additive
    translationSpace=BCS_BoneSpace
    alphaBoolBlend {blendOption=Linear}
    boneToModify {boneName=dyn_cape_shoulder_r_02}
    translation {y=-3}
   }
   n37 = "Transform (Modify) Bone - Bone: dyn_cape_mid_r_01"!(ComponentPose=@n36, Alpha=1, Rotation=@n34, Scale="1,1,1") -> (Pose) {
    rotationMode=BMM_Additive
    rotationSpace=BCS_BoneSpace
    translationMode=BMM_Additive
    translationSpace=BCS_BoneSpace
    alphaBoolBlend {blendOption=Linear}
    boneToModify {boneName=dyn_cape_mid_r_01}
    translation {y=-3}
   }
   n38 = "Transform (Modify) Bone - Bone: dyn_cape_mid_l_01"!(ComponentPose=@n37, Alpha=1, Rotation=@n34, Scale="1,1,1") -> (Pose) {
    rotationMode=BMM_Additive
    rotationSpace=BCS_BoneSpace
    translationMode=BMM_Additive
    translationSpace=BCS_BoneSpace
    alphaBoolBlend {blendOption=Linear}
    boneToModify {boneName=dyn_cape_mid_l_01}
    translation {y=-3}
   }
   n39 = Get BwdMovement() -> (BwdMovement:double)
   n40 = Trail controller - Bone: dyn_cape_mid_l_06!(ComponentPose=@n38, Alpha=@n39, RelaxationSpeedScale:float=1, FakeVelocity:Vector="0,0,0") -> (Pose) {
    bUsePlanarLimit=true
    chainLength=6
    debugLifeTime=0.1
    maxDeltaTime=0.0333
    relaxationSpeedScale=1.2
    alphaScaleBiasClamp {bInterpResult=true, interpSpeedDecreasing=20, interpSpeedIncreasing=5}
    rotationLimits [
     {limitMax=<180 180 180>}
     {limitMax=<180 180 180>}
     {limitMax=<180 180 180>}
     {limitMax=<180 180 180>}
     {limitMax=<180 180 180>}
     {limitMax=<180 180 180>}
    ]
    rotationOffsets [<0 0 0>, <0 0 0>, <0 0 0>, <0 0 0>, <0 0 0>, <0 0 0>]
    trailBone {boneName=dyn_cape_mid_l_06}
    trailRelaxationSpeed {editorCurveData {keys [<at 0 20 RCIM_Cubic>, {arriveTangent=-27.7668, interpMode=RCIM_Cubic, leaveTangent=-27.7668, tangentMode=RCTM_User, time=1, value=5}]}}
   }
   n41 = Trail controller - Bone: dyn_cape_mid_r_06!(ComponentPose=@n40, Alpha=@n39, RelaxationSpeedScale:float=1, FakeVelocity:Vector="0,0,0") -> (Pose) {
    bUsePlanarLimit=true
    chainLength=6
    debugLifeTime=0.1
    maxDeltaTime=0.0333
    relaxationSpeedScale=1.2
    alphaScaleBiasClamp {bInterpResult=true, interpSpeedDecreasing=20, interpSpeedIncreasing=5}
    rotationLimits [
     {limitMax=<180 180 180>}
     {limitMax=<180 180 180>}
     {limitMax=<180 180 180>}
     {limitMax=<180 180 180>}
     {limitMax=<180 180 180>}
     {limitMax=<180 180 180>}
    ]
    rotationOffsets [<0 0 0>, <0 0 0>, <0 0 0>, <0 0 0>, <0 0 0>, <0 0 0>]
    trailBone {boneName=dyn_cape_mid_r_06}
    trailRelaxationSpeed {editorCurveData {keys [<at 0 20 RCIM_Cubic>, {arriveTangent=-27.7668, interpMode=RCIM_Cubic, leaveTangent=-27.7668, tangentMode=RCTM_User, time=1, value=5}]}}
   }
   n42 = Trail controller - Bone: dyn_cape_shoulder_l_06!(ComponentPose=@n41, Alpha=@n39, RelaxationSpeedScale:float=1, FakeVelocity:Vector="0,0,0") -> (Pose) {
    bUsePlanarLimit=true
    chainLength=6
    debugLifeTime=0.1
    maxDeltaTime=0.0333
    relaxationSpeedScale=1.2
    alphaScaleBiasClamp {bInterpResult=true, interpSpeedDecreasing=20, interpSpeedIncreasing=5}
    rotationLimits [
     {limitMax=<180 180 180>}
     {limitMax=<180 180 180>}
     {limitMax=<180 180 180>}
     {limitMax=<180 180 180>}
     {limitMax=<180 180 180>}
     {limitMax=<180 180 180>}
    ]
    rotationOffsets [<0 0 0>, <0 0 0>, <0 0 0>, <0 0 0>, <0 0 0>, <0 0 0>]
    trailBone {boneName=dyn_cape_shoulder_l_06}
    trailRelaxationSpeed {editorCurveData {keys [<at 0 20 RCIM_Cubic>, {arriveTangent=-27.7668, interpMode=RCIM_Cubic, leaveTangent=-27.7668, tangentMode=RCTM_User, time=1, value=5}]}}
   }
   n43 = Trail controller - Bone: dyn_cape_shoulder_r_06!(ComponentPose=@n42, Alpha=@n39, RelaxationSpeedScale:float=1, FakeVelocity:Vector="0,0,0") -> (Pose) {
    bUsePlanarLimit=true
    chainLength=6
    debugLifeTime=0.1
    maxDeltaTime=0.0333
    relaxationSpeedScale=1.2
    alphaScaleBiasClamp {bInterpResult=true, interpSpeedDecreasing=20, interpSpeedIncreasing=5}
    rotationLimits [
     {limitMax=<180 180 180>}
     {limitMax=<180 180 180>}
     {limitMax=<180 180 180>}
     {limitMax=<180 180 180>}
     {limitMax=<180 180 180>}
     {limitMax=<180 180 180>}
    ]
    rotationOffsets [<0 0 0>, <0 0 0>, <0 0 0>, <0 0 0>, <0 0 0>, <0 0 0>]
    trailBone {boneName=dyn_cape_shoulder_r_06}
    trailRelaxationSpeed {editorCurveData {keys [<at 0 20 RCIM_Cubic>, {arriveTangent=-27.7668, interpMode=RCIM_Cubic, leaveTangent=-27.7668, tangentMode=RCTM_User, time=1, value=5}]}}
   }
   Save cached pose 'trailCache'(Pose=Component To Local(ComponentPose=@n43)) {cacheName=trailCache}
   n44 = "Constraint - cape_01_r (PP)"!(
    ComponentPose=Local To Component(LocalPose=Use cached pose 'initialPoseCache'())
    Alpha=1
    ConstraintWeights_0=0.28
    ConstraintWeights_1:float=0.72
   ) -> (Pose) {
    alphaBoolBlend {blendOption=Linear}
    boneToModify {boneName=cape_01_r}
    constraintSetup [{targetBone {boneName=spine_05}}, {targetBone {boneName=clavicle_r}}]
    constraintWeights [1,1]
   }
   n45 = "Constraint - cape_02_r (PP)"!(ComponentPose=@n44, Alpha=1, ConstraintWeights_0=0, ConstraintWeights_1:float=1) -> (Pose) {
    alphaBoolBlend {blendOption=Linear}
    boneToModify {boneName=cape_02_r}
    constraintSetup [{targetBone {boneName=spine_05}}, {targetBone {boneName=clavicle_r}}]
    constraintWeights [1,1]
   }
   n46 = "Constraint - cape_01_l (PP)"!(ComponentPose=@n45, Alpha=1, ConstraintWeights_0=0.1, ConstraintWeights_1:float=0.9) -> (Pose) {
    alphaBoolBlend {blendOption=Linear}
    boneToModify {boneName=cape_01_l}
    constraintSetup [{targetBone {boneName=spine_05}}, {targetBone {boneName=clavicle_l}}]
    constraintWeights [1,1]
   }
   n47 = "Constraint - cape_02_l (PP)"!(ComponentPose=@n46, Alpha=1, ConstraintWeights_0=0, ConstraintWeights_1:float=1) -> (Pose) {
    alphaBoolBlend {blendOption=Linear}
    boneToModify {boneName=cape_02_l}
    constraintSetup [{targetBone {boneName=spine_05}}, {targetBone {boneName=clavicle_l}}]
    constraintWeights [1,1]
   }
   n48 = Look At - Bone: cape_mid_01_l(ComponentPose=@n47, Alpha=1) {
    alphaBoolBlend {blendOption=Linear}
    boneToModify {boneName=cape_mid_01_l}
    lookAtLocation=<0 -25 -5>
    lookAtTarget {boneReference {boneName=pelvis}}
    lookAt_Axis {axis=<1 0>}
   }
   n49 = Look At - Bone: cape_mid_01_r(ComponentPose=@n48, Alpha=1) {
    alphaBoolBlend {blendOption=Linear}
    boneToModify {boneName=cape_mid_01_r}
    lookAtLocation=<0 -25 5>
    lookAtTarget {boneReference {boneName=pelvis}}
    lookAt_Axis {axis=<1 0>}
   }
   n50 = Look At - Bone: cape_02_l(ComponentPose=@n49, Alpha=1) {
    alphaBoolBlend {blendOption=Linear}
    boneToModify {boneName=cape_02_l}
    lookAtLocation=<0 -25 -20>
    lookAtTarget {boneReference {boneName=pelvis}}
    lookAt_Axis {axis=<1 0>}
   }
   n51 = Look At - Bone: cape_02_r(ComponentPose=@n50, Alpha=1) {
    alphaBoolBlend {blendOption=Linear}
    boneToModify {boneName=cape_02_r}
    lookAtLocation=<0 -25 21>
    lookAtTarget {boneReference {boneName=pelvis}}
    lookAt_Axis {axis=<1 0>}
   }
   n52 = Look At - Bone: cape_mid_04_l(ComponentPose=@n51, Alpha=1) {
    alphaBoolBlend {blendOption=Linear}
    boneToModify {boneName=cape_mid_04_l}
    lookAtLocation=<-75 -15 -10>
    lookAtTarget {boneReference {boneName=pelvis}}
    lookAt_Axis {axis=<1 0>}
   }
   n53 = Look At - Bone: cape_mid_04_r(ComponentPose=@n52, Alpha=1) {
    alphaBoolBlend {blendOption=Linear}
    boneToModify {boneName=cape_mid_04_r}
    lookAtLocation=<-75 -15 10>
    lookAtTarget {boneReference {boneName=pelvis}}
    lookAt_Axis {axis=<1 0>}
   }
   n54 = Look At - Bone: cape_mid_05_l(ComponentPose=@n53, Alpha=1) {
    alphaBoolBlend {blendOption=Linear}
    boneToModify {boneName=cape_mid_05_l}
    lookAtLocation=<-75 15 -11>
    lookAtTarget {boneReference {boneName=pelvis}}
    lookAt_Axis {axis=<1 0>}
   }
   n55 = Look At - Bone: cape_mid_05_r(ComponentPose=@n54, Alpha=1) {
    alphaBoolBlend {blendOption=Linear}
    boneToModify {boneName=cape_mid_05_r}
    lookAtLocation=<-75 15 11>
    lookAtTarget {boneReference {boneName=pelvis}}
    lookAt_Axis {axis=<1 0>}
   }
   n56 = Look At - Bone: cape_05_l(ComponentPose=@n55, Alpha=1) {
    alphaBoolBlend {blendOption=Linear}
    boneToModify {boneName=cape_05_l}
    lookAtLocation=<-75 -15 -25>
    lookAtTarget {boneReference {boneName=pelvis}}
    lookAt_Axis {axis=<1 0>}
   }
   n57 = Look At - Bone: cape_05_r(ComponentPose=@n56, Alpha=1) {
    alphaBoolBlend {blendOption=Linear}
    boneToModify {boneName=cape_05_r}
    lookAtLocation=<-75 -15 30>
    lookAtTarget {boneReference {boneName=pelvis}}
    lookAt_Axis {axis=<1 0>}
   }
   n58 = Look At - Bone: cape_06_l(ComponentPose=@n57, Alpha=1) {
    alphaBoolBlend {blendOption=Linear}
    boneToModify {boneName=cape_06_l}
    lookAtLocation=<-75 20 -25>
    lookAtTarget {boneReference {boneName=pelvis}}
    lookAt_Axis {axis=<1 0>}
   }
   n59 = Look At - Bone: cape_06_r(ComponentPose=@n58, Alpha=1) {
    alphaBoolBlend {blendOption=Linear}
    boneToModify {boneName=cape_06_r}
    lookAtLocation=<-75 20 30>
    lookAtTarget {boneReference {boneName=pelvis}}
    lookAt_Axis {axis=<1 0>}
   }
   n60 = "Transform (Modify) Bone - Bone: cape_02_l"!(ComponentPose=@n59, Alpha=1, Rotation="0,0,0", Scale="1,1,1") -> (Pose) {
    translationMode=BMM_Additive
    translationSpace=BCS_BoneSpace
    alphaBoolBlend {blendOption=Linear}
    boneToModify {boneName=cape_02_l}
    translation {y=-3}
   }
   n61 = "Transform (Modify) Bone - Bone: cape_02_r"!(ComponentPose=@n60, Alpha=1, Rotation="0,0,0", Scale="1,1,1") -> (Pose) {
    translationMode=BMM_Additive
    translationSpace=BCS_BoneSpace
    alphaBoolBlend {blendOption=Linear}
    boneToModify {boneName=cape_02_r}
    translation {y=-3}
   }
   n62 = "Transform (Modify) Bone - Bone: cape_mid_01_r"!(ComponentPose=@n61, Alpha:float=1, Rotation="0,0,0", Scale="1,1,1") -> (Pose) {
    translationMode=BMM_Additive
    translationSpace=BCS_BoneSpace
    alphaBoolBlend {blendOption=Linear}
    boneToModify {boneName=cape_mid_01_r}
    translation {y=-3}
   }
   n63 = "Transform (Modify) Bone - Bone: cape_mid_01_l"!(ComponentPose=@n62, Alpha=1, Rotation="0,0,0", Scale="1,1,1") -> (Pose) {
    translationMode=BMM_Additive
    translationSpace=BCS_BoneSpace
    alphaBoolBlend {blendOption=Linear}
    boneToModify {boneName=cape_mid_01_l}
    translation {y=-3}
   }
   n64 = Trail controller - Bone: cape_mid_07_l!(ComponentPose=@n63, Alpha=1, RelaxationSpeedScale:float=2, FakeVelocity:Vector="0,0,0") -> (Pose) {
    bActorSpaceFakeVel=true
    bShowTrailLocation=true
    bUsePlanarLimit=true
    chainLength=6
    debugLifeTime=0.1
    maxDeltaTime=0.0333
    relaxationSpeedScale=1.2
    alphaScaleBiasClamp {bInterpResult=true, interpSpeedDecreasing=25, interpSpeedIncreasing=25}
    baseJoint {boneName=root}
    rotationLimits [
     {limitMax=<180 180 180>}
     {limitMax=<180 180 180>}
     {limitMax=<180 180 180>}
     {limitMax=<180 180 180>}
     {limitMax=<180 180 180>}
     {limitMax=<180 180 180>}
    ]
    rotationOffsets [<0 0 0>, <0 0 0>, <0 0 0>, <0 0 0>, <0 0 0>, <0 0 0>]
    trailBone {boneName=cape_mid_07_l}
    trailRelaxationSpeed {editorCurveData {keys [<at 0 15 RCIM_Linear>,<at 1 6 RCIM_Linear>]}}
   }
   n65 = Trail controller - Bone: cape_mid_07_r!(ComponentPose=@n64, Alpha=1, RelaxationSpeedScale:float=2, FakeVelocity:Vector="0,0,0") -> (Pose) {
    bActorSpaceFakeVel=true
    bShowTrailLocation=true
    bUsePlanarLimit=true
    chainLength=6
    debugLifeTime=0.1
    maxDeltaTime=0.0333
    relaxationSpeedScale=1.2
    alphaScaleBiasClamp {bInterpResult=true, interpSpeedDecreasing=25, interpSpeedIncreasing=25}
    baseJoint {boneName=root}
    rotationLimits [
     {limitMax=<180 180 180>}
     {limitMax=<180 180 180>}
     {limitMax=<180 180 180>}
     {limitMax=<180 180 180>}
     {limitMax=<180 180 180>}
     {limitMax=<180 180 180>}
    ]
    rotationOffsets [<0 0 0>, <0 0 0>, <0 0 0>, <0 0 0>, <0 0 0>, <0 0 0>]
    trailBone {boneName=cape_mid_07_r}
    trailRelaxationSpeed {editorCurveData {keys [<at 0 15 RCIM_Linear>,<at 1 6 RCIM_Linear>]}}
   }
   n66 = Trail controller - Bone: cape_08_l!(ComponentPose=@n65, Alpha=1, RelaxationSpeedScale:float=2, FakeVelocity:Vector="0,0,0") -> (Pose) {
    bActorSpaceFakeVel=true
    bShowTrailLocation=true
    bUsePlanarLimit=true
    chainLength=6
    debugLifeTime=0.1
    maxDeltaTime=0.0333
    relaxationSpeedScale=1.2
    alphaScaleBiasClamp {bInterpResult=true, interpSpeedDecreasing=25, interpSpeedIncreasing=25}
    baseJoint {boneName=root}
    rotationLimits [
     {limitMax=<180 180 180>}
     {limitMax=<180 180 180>}
     {limitMax=<180 180 180>}
     {limitMax=<180 180 180>}
     {limitMax=<180 180 180>}
     {limitMax=<180 180 180>}
    ]
    rotationOffsets [<0 0 0>, <0 0 0>, <0 0 0>, <0 0 0>, <0 0 0>, <0 0 0>]
    trailBone {boneName=cape_08_l}
    trailRelaxationSpeed {editorCurveData {keys [<at 0 15 RCIM_Linear>,<at 1 6 RCIM_Linear>]}}
   }
   n67 = Trail controller - Bone: cape_08_r!(ComponentPose=@n66, Alpha=1, RelaxationSpeedScale:float=2, FakeVelocity:Vector="0,0,0") -> (Pose) {
    bActorSpaceFakeVel=true
    bShowTrailLocation=true
    bUsePlanarLimit=true
    chainLength=6
    debugLifeTime=0.1
    maxDeltaTime=0.0333
    relaxationSpeedScale=1.2
    alphaScaleBiasClamp {bInterpResult=true, interpSpeedDecreasing=25, interpSpeedIncreasing=25}
    baseJoint {boneName=root}
    rotationLimits [
     {limitMax=<180 180 180>}
     {limitMax=<180 180 180>}
     {limitMax=<180 180 180>}
     {limitMax=<180 180 180>}
     {limitMax=<180 180 180>}
     {limitMax=<180 180 180>}
    ]
    rotationOffsets [<0 0 0>, <0 0 0>, <0 0 0>, <0 0 0>, <0 0 0>, <0 0 0>]
    trailBone {boneName=cape_08_r}
    trailRelaxationSpeed {editorCurveData {keys [<at 0 15 RCIM_Linear>,<at 1 6 RCIM_Linear>]}}
   }
   Save cached pose 'dbnoCache'(Pose=Component To Local(ComponentPose=@n67)) {cacheName=dbnoCache}
   n68 = RigidBody!(ComponentPose=Local To Component(LocalPose=Use cached pose 'initialPoseCache'()), Alpha, OverrideWorldGravity=@n12, ExternalForce) -> (Pose) {
    bClampLinearTranslationLimitToRefPose=true
    bOverrideWorldGravity=true
    overridePhysicsAsset=&F_MED_Monarch_new_cape_Physics
    simulationSpace=BaseBoneSpace
    alphaBoolBlend {blendOption=Linear}
    baseBoneRef {boneName=root}
    componentAppliedLinearAccClamp=<3000 3000 3000>
    componentLinearAccScale {z=0.3}
    componentLinearVelScale {z=0.3}
    overrideWorldGravity {z=-980}
    simSpaceSettings {
     maxAngularAcceleration=1000
     maxAngularVelocity=100
     maxLinearAcceleration=100
     maxLinearVelocity=100
     velocityScaleZ=0
     worldAlpha=1
     externalLinearDragV=<1 1>
    }
   }
   n69 = RigidBody!(
    ComponentPose=Local To Component(LocalPose=Use cached pose 'initialPoseCache'())
    Alpha
    OverrideWorldGravity=@n12
    ExternalForce
    ComponentLinearAccScale:Vector=@n13
    ComponentLinearVelScale:Vector=@n14
   ) -> (Pose) {
    bClampLinearTranslationLimitToRefPose=true
    bOverrideWorldGravity=true
    overridePhysicsAsset=&F_MED_Monarch_new_cape_Physics
    simulationSpace=BaseBoneSpace
    alphaBoolBlend {blendOption=Linear}
    baseBoneRef {boneName=root}
    componentAppliedLinearAccClamp=<3000 3000 3000>
    componentLinearAccScale {z=0.5}
    componentLinearVelScale {z=0.5}
    overrideWorldGravity {z=-980}
    simSpaceSettings {
     maxAngularAcceleration=500
     maxAngularVelocity=2
     maxLinearAcceleration=400
     maxLinearVelocity=400
     velocityScaleZ=0
     worldAlpha=1
     externalLinearDragV=<1 1>
    }
   }
   Save cached pose 'rigidChche'(Pose=Blend Poses by bool(BlendPose_0=Component To Local(ComponentPose=@n68), BlendPose_1=Component To Local(ComponentPose=@n69), bActiveValue=Get bIsInFrontEnd()) {blendType=Linear}) {cacheName=rigidChche}
   Use cached pose 'dbnoCache'()
   Get ADS rigid() -> (ADS rigid:double)
  ]}
 EventGraph {nodes [
   Event BlueprintUpdateAnimation()
   > n1 = Sequence!(execute) -> (then_0, then_1, then_2:Exec, then_3:Exec, then_4:Exec, then_5:Exec, then_6:Exec, then_7:Exec, then_8:Exec, then_9:Exec, then_10:Exec)
   n2 = GetPartSkeletalMeshComponent()
   > n3 = TransformFromBoneSpace!(
    execute
    self:Skinned Mesh Component Object Reference=@n2
    BoneName:Name=lowerarm_r
    InPosition:Vector="0, 0, 0"
    InRotation:Rotator="0, 0, 0"
   ) -> (then, OutPosition:Vector, OutRotation:Rotator)
   > n4 = TransformFromBoneSpace(self=@n2, BoneName=head, InPosition="0, 0, 0", InRotation="0, 0, 0")
   > Set armForwardUpR(armForwardUpR:double=MapRangeClamped(Value="Distance (Vector)"!(V1:Vector=@n3, V2:Vector=@n4) -> (ReturnValue:double), InRangeA=40, InRangeB=20)) -> (Output_Get:double)
   n5 = GetPartSkeletalMeshComponent()
   > n6 = TransformFromBoneSpace(self=@n5, BoneName=lowerarm_l, InPosition="0, 0, 0", InRotation="0, 0, 0")
   > n7 = TransformFromBoneSpace(self=@n5, BoneName=head, InPosition="0, 0, 0", InRotation="0, 0, 0")
   > Set armForwardUpL(armForwardUpL:double=MapRangeClamped(Value="Distance (Vector)"(V1=@n6, V2=@n7), InRangeA=42, InRangeB=30)) -> (Output_Get:double)
   n8 = Gravity Override State Function DEPRECATED(
    execute=@n1.then_1
    CharacterPartAnimInstance=Self-Reference()
    GameTimeInSeconds=GetGameTimeInSeconds()
    GravityOverrideParams=Get Gravity Override Params()
   )
   > Set GravityOverride(GravityOverride=@n8)
   n9 = Get bRigidBodyEnabled()
   n10 = Branch(execute=@n1.then_10, Condition=AND Boolean(A=Get bIsSkyDiving(), B=NOT Boolean(A=@n9)))
   > Set skydive rigids(skydive rigids:bool=true) -> (Output_Get:bool)
   Set skydive rigids(execute=@n10.else, skydive rigids=false)
   n11 = ComplexSimSpaceFunction DEPRECATED!(
    execute=@n1.then_2
    CharacterPartAnimInstance:Custom Character Part Anim Instance Object Reference=Self-Reference()
    InterpSpeedMax:double=9
    PawnSpeedMax:double=20
    RInterpErrorTolerance:double=1
    DeltaTime:double=0
    DebugDraw:bool=false
    sim_space_rotation_in:Rotator=Get sim_space_rotation()
    inComponentSpaceIn:bool=Get InComponentSpace() -> (InComponentSpace:bool)
    inComponentSpaceLastFrameIn:bool=Get InComponentSpaceLastFrame() -> (InComponentSpaceLastFrame:bool)
    lockedOnSimSpaceIn:bool=Get LockedOnSimSpace() -> (LockedOnSimSpace:bool)
    InterpSpeedIn:double=Get InterpSpeed() -> (InterpSpeed:double)
   ) -> (
    then
    sim_space_rotation:Rotator
    inComponentSpaceOut:bool
    inComponentSpaceLastFrameOut:bool
    lockedOnSimSpaceOut:bool
    InterpSpeedOut:double
   )
   > Set sim_space_rotation(sim_space_rotation:Rotator=@n11) -> (Output_Get:Rotator)
   > Set InComponentSpace(InComponentSpace:bool=@n11.inComponentSpaceOut) -> (Output_Get:bool)
   > Set InComponentSpaceLastFrame(InComponentSpaceLastFrame:bool=@n11.inComponentSpaceLastFrameOut) -> (Output_Get:bool)
   > Set LockedOnSimSpace(LockedOnSimSpace:bool=@n11.lockedOnSimSpaceOut) -> (Output_Get:bool)
   > Set InterpSpeed(InterpSpeed:double=@n11.InterpSpeedOut) -> (Output_Get:double)
   n12 = Break Fort Anim Input Slope Info!(FortAnimInput_SlopeInfo:"Fort Anim Input Slope Info Structure (by ref)"=Get SlopeInfoInput() -> (SlopeInfoInput:Fort Anim Input Slope Info Structure)) -> (RootSlopeTranslation:Vector=~, LeftFootSlopeTranslation:Vector=~, RightFootSlopeTranslation:Vector=~, SlopeRotation:Rotator=~) {is=title K2Node_BreakStruct}
   Set slope_root_trans(execute=@n1.then_3, slope_root_trans:Vector=@n12) -> (Output_Get:Vector)
   > Set slope_foot_rot(slope_foot_rot:Rotator=@n12.SlopeRotation) -> (Output_Get:Rotator)
   n13 = PlayerMovement(execute=@n1.then_4, AnimInstance=Self-Reference(), XDotProductCoefficient=1, YDotProductCoefficient=1, ZDotProductCoefficient=1)
   n14 = "Absolute (Float)"(A=@n13)
   > Set BwdMovement(BwdMovement:double=float * float(
     A=MapRangeClamped(Value=@n14, InRangeA=0, InRangeB=1, OutRangeA=1, OutRangeB=0.4)
     B=MapRangeClamped(Value=@n13.YMovement, InRangeA=0, InRangeB=-1, OutRangeA=1, OutRangeB=0)
    )) -> (Output_Get:double)
   n15 = Branch(execute=@n1.then_5, Condition=Get bIsSkyDiving())
   > n16 = Branch(Condition="float > float"(A=Get PawnVelocityZ(), B=50) {is=title K2Node_CallFunction})
   > Set BwdMovement(BwdMovement=float * float(A=Get BwdMovement(), B=0))
   Set BwdMovement(execute=[@n15.else, @n16.else], BwdMovement=float * float(A=Get BwdMovement(), B=1))
   n17 = Cast To FortPlayerPawn(execute=@n1.then_6, Object=TryGetPawnOwner())
   n18 = GetDisplayName(Object=GetSkeletalMeshForPartType(self=@n17))
   n19 = MatchesWildcard(SourceString=@n18, Wildcard=*Lemon_Lime*, SearchCase=CaseSensitive)
   n20 = MatchesWildcard(SourceString=@n18, Wildcard=*F_MED*)
   > Set isFemaleSkin(isFemaleSkin:double="To Float (Boolean)"(InBool=OR Boolean(A=@n20, B=@n19))) -> (Output_Get:double)
   n21 = Branch(execute=@n1.then_7, Condition=Get bIsTargeting())
   > Set ADS rigid(ADS rigid:double=0.7) -> (Output_Get:double)
   > Set ADS Accel(ADS Accel:Vector="0, 0,0.3") -> (Output_Get:Vector)
   > Set ADS Vel(ADS Vel:Vector="0, 0,0.3") -> (Output_Get:Vector)
   Set ADS rigid(execute=@n21.else, ADS rigid=1)
   > Set ADS Accel(ADS Accel="0, 0,0.5")
   > Set ADS Vel(ADS Vel="0, 0,0.5")
   Branch(
    execute=@n1.then_8
    Condition=CharacterItemDefinitionHasAnyMetaTag(Tags=Make Array!("[0]"=(TagName="Cosmetics.MetaTags.Monarch"), "[1]":Gameplay Tag Structure) -> (Array))
   )
   > n22 = Branch(Condition=Get bIsParachuteOpened())
   > Set ADS Accel(ADS Accel="0, 0,0.2")
   > Set ADS Vel(ADS Vel="0, 0,0.2")
   Set ADS Accel(execute=@n22.else, ADS Accel="0.3,0.3,0.3")
   > Set ADS Vel(ADS Vel="0.3,0.3,0.3")
   n23 = Get bRigidBodyEnabled()
   n24 = Branch(execute=@n1.then_9, Condition=AND Boolean(A=Get bIsInFrontEnd(), B=NOT Boolean(A=@n23)))
   > Set front end no rigid(front end no rigid:bool=true) -> (Output_Get:bool)
   Set front end no rigid(execute=@n24.else, front end no rigid=false)
   Event BlueprintInitializeAnimation()
   n25 = TryGetPawnOwner()
   > Is Valid(InputObject=@n25)
   > n26 = Cast To FortPlayerPawnAthena(Object=@n25)
   > Set player_pawn(player_pawn:Fort Player Pawn Athena Object Reference=@n26) -> (Output_Get:Fort Player Pawn Athena Object Reference)
   > Set Skel_Mesh(Skel_Mesh=GetOwnerSkeletalMeshComponent())
   > Set InterpSpeed(InterpSpeed=Get InterpSpeedMax() -> (InterpSpeedMax:double))
   Event OnEndCrouching()
   > Set crouch alpha(crouch alpha:double=0) -> (Output_Get:double)
   > Set crouch alpha inv(crouch alpha inv:double=1) -> (Output_Get:double)
   Event OnBeginCrouching()
   > Set crouch alpha(crouch alpha=1)
   > Set crouch alpha inv(crouch alpha inv=0)
   Event OnBeginSkydiving()
   > Set SkydivingAlpha(SkydivingAlpha:Rotator="0,-30, 0") -> (Output_Get:Rotator)
   Event OnEndSkydiving()
   > Set SkydivingAlpha(SkydivingAlpha="0, 0, 0")
   TryGetPawnOwner()
   n27 = Get Skel_Mesh()
   GetAlphaFromReferenceTransformDelta(SkeletalMeshComponent=@n27, BoneName=thigh_l, BaseName=pelvis, InputRange=(X=-20,Y=-100), TwistAxis="1,0,0")
   TryGetPawnOwner()
   GetAlphaFromReferenceTransformDelta(SkeletalMeshComponent=@n27, BoneName=thigh_r, BaseName=pelvis, InputRange=(X=-20,Y=-100), TwistAxis="1,0,0")
   Make Array!(
    "[0]"=(TagName="Cosmetics.MetaTags.Monarch")
    "[1]"
    "[2]":Gameplay Tag Structure
    "[3]":Gameplay Tag Structure
    "[4]":Gameplay Tag Structure
    "[5]":Gameplay Tag Structure
    "[6]":Gameplay Tag Structure
    "[7]":Gameplay Tag Structure
   ) -> (Array)
   Get ADS Accel()
  ]}
]
variables [
 ADS Accel {type=Vector}
 ADS rigid {type=number}
 ADS Vel {type=Vector}
 armForwardUpL {type=number}
 armForwardUpR {type=number}
 BwdMovement {type=number}
 crouch alpha {type=number}
 crouch alpha inv {type=number}
 front end no rigid {type=boolean}
 Gravity Override Params {
  type=GravityOverrideParamsStruct
  default {
   skydive_DivingDown {gravityOverride=<0 0 0>}
   skydive_FloatingDown {gravityOverride=<0 0 0>}
   skydive_GoingUp {gravityOverride=<0 -300 -100>, windOffset=<0 -600 400>}
   skydive_Parachuting {gravityOverride=<0 0 -980>}
   swimming {gravityOverride=<0 0 -580>}
  }
 }
 GravityOverride {type=Vector}
 InComponentSpace {type=boolean}
 InComponentSpaceLastFrame {type=boolean}
 InterpSpeed {type=number}
 InterpSpeedMax {type=number}
 isFemaleSkin {type=number}
 LockedOnSimSpace {type=boolean}
 player_pawn {type=FortPlayerPawnAthena}
 sim_space_rotation {type=Rotator}
 Skel_Mesh {type=SkeletalMeshComponent}
 skydive rigids {type=boolean}
 SkydivingAlpha {type=Rotator}
 slope_foot_rot {type=Rotator}
 slope_root_trans {type=Vector}
]
